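# Physics asset for the Bountress female medium hair test mesh, placed among the cosmetic hair material examples.
# It holds a single body on the head bone, made of one angled capsule offset from the bone origin.
# It defines no constraints, so there are no jointed or chained bodies.
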PhysicsAsset /MiscReferencersOfCosmetics/Global/Materials/Cosmetics/Hair/Examples/Bountress/Mesh/F_MED_Bountress_HairTest_PhysicsAsset
bodies {head [{shapeType=Capsule, length=12.2362, radius=10.1276, center=<3.6387 1.2397 5.4583>, rotation=<rot 0 -115.4293 -62.9651>}]}
constraints []
